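; Time-of-day condition for the Sparks lobby that is true when volumetric fog counts as off, using the shadow quality setting as a proxy.
; Evaluate logs and returns whether the user's shadow quality level is 1 or lower.
; Setup binds the condition's value-changed broadcast to the Fortnite user settings' shadow quality change event, so changes to that setting are signalled without polling.
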(fn Evaluate
  (bind _returnvalue (Settings|GetGameUserSettings))
  (Fortnite|FortShippingLog self (Utilities|String|Append "TODM Condition: Volumetric Fog = " (Utilities|String|ToString(Boolean) (<= (Settings|GetShadowQuality _returnvalue) 1))))
  (return (<= (Settings|GetShadowQuality _returnvalue) 1)))

(fn SetupOnConditionValueChanged
  (EventDispatchers|BindEventtoOnShadowQualityChanged (Settings|GetFortGameUserSettings) (EventDispatchers|CreateEvent :!event BroadcastOnConditionValueChanged)))
